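; A QA helper Blueprint, BP_ConsoleCommand_BeginPlay, that runs a console command shortly after play begins.
; On BeginPlay it waits through a five-second delay, then executes the string held in its ConsoleCommand variable.
; The construction script is empty.
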(event EventBeginPlay
  (Utilities|FlowControl|Delay 5.0)
  (Development|ExecuteConsoleCommand (Variables|Default|GetConsoleCommand)))

(fn ConstructionScript)
