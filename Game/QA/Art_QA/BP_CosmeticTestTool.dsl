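; QA art test actor that forces cosmetics, variants and variant colors on players, or resets forced cosmetics.
; Trigger overlap, after a gated delay against accidental activation, applies the configured item definitions and variant infos to the pawn's controller.
; In variant or remove-backpack mode, interacting instead runs ForceAthenaCosmeticItem console commands, with variant changes targeting all player controllers.
; The construction script sets labels and base color, stripping components each mode does not need.
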(event OnComponentBeginOverlap(CollisionTrigger)
  (params (OverlappedComponent :type Primitive%20Component%20Object%20Reference) (OtherActor :type Actor%20Object%20Reference) (OtherComp :type Primitive%20Component%20Object%20Reference) (OtherBodyIndex :type Integer) (bFromSweep :type Boolean) (SweepResult :type Hit%20Result%20Structure%20%28by%20ref%29))
  (bind _backpack (Variables|ItemDefinition|GetBackpack))
  (bind _wrap (Variables|ItemDefinition|GetWrap))
  (bind _pickaxe (Variables|ItemDefinition|GetPickaxe))
  (bind _glider (Variables|ItemDefinition|GetGlider))
  (bind _character (Variables|ItemDefinition|GetCharacter))
  ;[1 1 1] activation delay to avoid accidental trigger
  (Utilities|FlowControl|Sequence
    (:then_0
      (goto _label Open))
    (:then_1
      (Utilities|FlowControl|Delay 0.5)))
  (Utilities|FlowControl|Gate false :!label _label
    (:Exit
      (bind _asfort_pawn (Utilities|Casting|CastToFortPawn OtherActor)
        (:then
          (Variables|Default|SetOvelappedController (Pawn|GetController _asfort_pawn)))
        (:CastFailed
          _))
      (Utilities|FlowControl|Sequence
        (:then_0
          (if (not (Variables|Default|GetbResetForcedAthenaCosmetics))
            ;[1 1 1] apply cosmetic data if valid
            (Utilities|FlowControl|Sequence
              (:then_0
                (if (Utilities|IsValid _character :!class /Script/Engine.KismetSystemLibrary)
                  (Items|ForceAthenaCosmetic _character)))
              (:then_1
                (if (Utilities|IsValid _backpack :!class /Script/Engine.KismetSystemLibrary)
                  (Items|ForceAthenaCosmetic _backpack)
                  (else
                    (bind _output_get (Variables|ItemDefinition|SetBackpack "/Game/Athena/Items/Cosmetics/Backpacks/BID_Random.BID_Random"))
                    (Items|ForceAthenaCosmetic _output_get))))
              (:then_2
                (if (Utilities|IsValid _pickaxe :!class /Script/Engine.KismetSystemLibrary)
                  (Items|ForceAthenaCosmetic _pickaxe)))
              (:then_3
                (if (Utilities|IsValid _glider :!class /Script/Engine.KismetSystemLibrary)
                  (Items|ForceAthenaCosmetic _glider)))
              (:then_4
                (if (Utilities|IsValid _wrap :!class /Script/Engine.KismetSystemLibrary)
                  (Items|ForceAthenaCosmetic _wrap))))
            (else
              ;[1 1 1] Reset Forced Athena Cosmetics
              (Items|ResetForcedAthenaCosmetics)
              (bind _ovelappedcontroller (Variables|Default|GetOvelappedController)
                (:then
                  (Cosmetic|ClearForcedCosmeticVariants _ovelappedcontroller))
                (:else)))))
        (:then_1
          ;[1 1 1] Apply Cosmetic Channel Variances
          (bind _ovelappedcontroller_2 (Variables|Default|GetOvelappedController)
            (:then
              (Cosmetic|ClearForcedCosmeticVariants _ovelappedcontroller_2)
              (Cosmetic|ForceCosmeticVariants _ovelappedcontroller_2 (Variables|ItemDefinition|VariantInfo|GetVariantInfosToApplyOnOverlap))
              ;[1 1 1] Apply the Forced Cosmetic Variance Color
              (if (Variables|ItemDefinition|ForcedVarianceColor|GetApplyForcedVarianceColor)
                (bind _ovelappedcontroller_3 (Variables|Default|GetOvelappedController)
                  (:then
                    (Cosmetic|ForceCosmeticVariantstoMatchReferenceColor _ovelappedcontroller_3 (Math|Conversions|ToLinearColor(Color) (Variables|ItemDefinition|ForcedVarianceColor|GetForcedVarianceColor))))
                  (:else))))
            (:else)))))))

(event OnComponentEndOverlap(CollisionTrigger)
  (params (OverlappedComponent :type Primitive%20Component%20Object%20Reference) (OtherActor :type Actor%20Object%20Reference) (OtherComp :type Primitive%20Component%20Object%20Reference) (OtherBodyIndex :type Integer))
  (bind _backpack (Variables|ItemDefinition|GetBackpack))
  (bind _wrap (Variables|ItemDefinition|GetWrap))
  (bind _pickaxe (Variables|ItemDefinition|GetPickaxe))
  (bind _glider (Variables|ItemDefinition|GetGlider))
  (bind _character (Variables|ItemDefinition|GetCharacter))
  ;[1 1 1] activation delay to avoid accidental trigger
  (Utilities|FlowControl|Gate false
    (:Exit
      (bind _asfort_pawn (Utilities|Casting|CastToFortPawn (stale OnComponentBeginOverlap(CollisionTrigger) OtherActor))
        (:then
          (Variables|Default|SetOvelappedController (Pawn|GetController _asfort_pawn)))
        (:CastFailed
          _))
      (Utilities|FlowControl|Sequence
        (:then_0
          (if (not (Variables|Default|GetbResetForcedAthenaCosmetics))
            ;[1 1 1] apply cosmetic data if valid
            (Utilities|FlowControl|Sequence
              (:then_0
                (if (Utilities|IsValid _character :!class /Script/Engine.KismetSystemLibrary)
                  (Items|ForceAthenaCosmetic _character)))
              (:then_1
                (if (Utilities|IsValid _backpack :!class /Script/Engine.KismetSystemLibrary)
                  (Items|ForceAthenaCosmetic _backpack)
                  (else
                    (bind _output_get (Variables|ItemDefinition|SetBackpack "/Game/Athena/Items/Cosmetics/Backpacks/BID_Random.BID_Random"))
                    (Items|ForceAthenaCosmetic _output_get))))
              (:then_2
                (if (Utilities|IsValid _pickaxe :!class /Script/Engine.KismetSystemLibrary)
                  (Items|ForceAthenaCosmetic _pickaxe)))
              (:then_3
                (if (Utilities|IsValid _glider :!class /Script/Engine.KismetSystemLibrary)
                  (Items|ForceAthenaCosmetic _glider)))
              (:then_4
                (if (Utilities|IsValid _wrap :!class /Script/Engine.KismetSystemLibrary)
                  (Items|ForceAthenaCosmetic _wrap))))
            (else
              ;[1 1 1] Reset Forced Athena Cosmetics
              (Items|ResetForcedAthenaCosmetics)
              (bind _ovelappedcontroller (Variables|Default|GetOvelappedController)
                (:then
                  (Cosmetic|ClearForcedCosmeticVariants _ovelappedcontroller))
                (:else)))))
        (:then_1
          ;[1 1 1] Apply Cosmetic Channel Variances
          (bind _ovelappedcontroller_2 (Variables|Default|GetOvelappedController)
            (:then
              (Cosmetic|ClearForcedCosmeticVariants _ovelappedcontroller_2)
              (Cosmetic|ForceCosmeticVariants _ovelappedcontroller_2 (Variables|ItemDefinition|VariantInfo|GetVariantInfosToApplyOnOverlap))
              ;[1 1 1] Apply the Forced Cosmetic Variance Color
              (if (Variables|ItemDefinition|ForcedVarianceColor|GetApplyForcedVarianceColor)
                (bind _ovelappedcontroller_3 (Variables|Default|GetOvelappedController)
                  (:then
                    (Cosmetic|ForceCosmeticVariantstoMatchReferenceColor _ovelappedcontroller_3 (Math|Conversions|ToLinearColor(Color) (Variables|ItemDefinition|ForcedVarianceColor|GetForcedVarianceColor))))
                  (:else))))
            (:else)))))))

(event EventBeginPlay
  ;[1 1 1] destroys interact box if no variant data is set
  (if (Variables|Default|GetIsVariant)
    _
    (elif (Variables|ItemDefinition|GetRemoveBackpack)
      _
      (else
        (Components|DestroyComponent (Variables|Default|GetInteractBox))))))

(event Interaction|EventBlueprintOnInteract
  (params (InteractingPawn :type Fort%20Pawn%20Object%20Reference) (InteractionBeingAttempted :type EInteractionBeingAttempted%20Enum :default "FirstInteraction"))
  ;[1 1 1] variant selection
  (CallFunction|VariantSelected))

(event Custom|VariantSelected
  (bind _variantinfotoapplyonselection (Variables|ItemDefinition|VariantInfo|GetVariantInfoToApplyOnSelection))
  (bind (_ownedvarianttags _itemvariantisusedfor _customdata _variantchanneltag _activevarianttag) (Utilities|Struct|BreakMcpVariantChannelInfo _variantinfotoapplyonselection))
  ;[1 1 1] variant selection
  (Networking|SwitchHasAuthority
    (:Authority
      (if (Variables|Default|GetIsVariant)
        (bind _asfort_item_definition (Utilities|Casting|CastToFortItemDefinition _itemvariantisusedfor))
        (Development|ExecuteConsoleCommand (Utilities|String|Append "ForceAthenaCosmeticItem " (Item|GetPersistentName _asfort_item_definition)) (Fortnite|GetFortPlayerControllerfromActor (stale Interaction|EventBlueprintOnInteract InteractingPawn)))
        (Class|FortPlayerController|ChangeActiveVariantforCosmeticItem (Fortnite|GetAllFortPlayerControllers true) (Utilities|String|StringToName (Item|GetPersistentName _asfort_item_definition)) _variantchanneltag _activevarianttag 0)
        (elif (Variables|ItemDefinition|GetRemoveBackpack)
          (Development|ExecuteConsoleCommand "ForceAthenaCosmeticItem AthenaBackpack:BID_Random" (Fortnite|GetFortPlayerControllerfromActor (stale Interaction|EventBlueprintOnInteract InteractingPawn))))))))

(fn BlueprintCanInteract
  (params (InteractingPawn :type Fort%20Pawn%20Object%20Reference) (InteractionBeingAttempted :type EInteractionBeingAttempted%20Enum :default "FirstInteraction") (InteractionType :type TInteractionType%20Enum :default "IT_NoInteraction"))
  (return true))

(fn BlueprintGetInteractionString
  (params (InteractingPawn :type Fort%20Pawn%20Object%20Reference) (InteractionBeingAttempted :type EInteractionBeingAttempted%20Enum :default "FirstInteraction"))
  (return "NSLOCTEXT(\"[190B08FA4B364C2A6544B5BB1A6B5D22]\", \"8014040D48B7D41D040BCCB3BB538571\", \"Hold E to Equip\")"))

(fn CreateAssetNameString
  (params (asset :type Fort%20Item%20Definition%20Object%20Reference))
  (bind _returnvalue (Utilities|GetDisplayName asset))
  (bind _returnvalue_2 (ItemDefinition|GetDisplayName asset))
  (bind _returnvalue_3 (Utilities|String|ToString(Text) _returnvalue_2))
  (bind _returnvalue_4 (Utilities|String|FindSubstring _returnvalue "_" false true))
  (bind _returnvalue_5 (+ _returnvalue_4 1))
  (bind _returnvalue_6 (Utilities|String|GetSubstring _returnvalue _returnvalue_5 30))
  (if (Variables|Default|GetbResetForcedAthenaCosmetics)
    (return "NSLOCTEXT(\"[190B08FA4B364C2A6544B5BB1A6B5D22]\", \"C6873CD44004B5DA7B0BF28FD4FC1F6B\", \"Clear Cosmetic Overrides\")")
    (else
      (Utilities|IsValid asset
        (:Is%20Valid
          (bind _output_get (Variables|Default|SetIsTBD (Utilities|String|Contains _returnvalue_3 "TBD")))
          (if _output_get
            (return (Utilities|Text|ToText(String) (Utilities|String|Append _returnvalue_6 "
(TBD)")))
            (else
              (return (Utilities|Text|ToText(String) (Utilities|String|Append _returnvalue_6 "
(" _returnvalue_3 ")"))))))
        (:Is%20Not%20Valid
          (return "NSLOCTEXT(\"[190B08FA4B364C2A6544B5BB1A6B5D22]\", \"EE7BCCD34894BC32CA7822BC167CB6CE\", \" \")"))))))

(fn ConstructionScript
  (bind _variantinfotoapplyonselection (Variables|ItemDefinition|VariantInfo|GetVariantInfoToApplyOnSelection))
  (bind (_ownedvarianttags _itemvariantisusedfor _customdata _variantchanneltag _activevarianttag) (Utilities|Struct|BreakMcpVariantChannelInfo _variantinfotoapplyonselection))
  (bind _applyforcedvariancecolor (Variables|ItemDefinition|ForcedVarianceColor|GetApplyForcedVarianceColor))
  (Utilities|FlowControl|Sequence
    (:then_0
      (LocalVariables|SetL_ActiveVariantTag _activevarianttag)
      (if (Utilities|IsValid _itemvariantisusedfor :!class /Script/Engine.KismetSystemLibrary)
        (Variables|Default|SetIsVariant true)
        (elif (Math|Boolean|Equal(Boolean) (Math|Boolean|ORBoolean (Variables|ItemDefinition|GetShowWrapID) (Variables|ItemDefinition|GetShowPickaxeID) (Variables|ItemDefinition|GetShowGliderID) (Variables|ItemDefinition|GetShowBackpackID) (Variables|ItemDefinition|GetRemoveBackpack)) true)
          _
          (else
            (Utilities|FlowControl|Sequence :!label _label
              (:then_0
                (if (Variables|Default|GetShowCharacterID)
                  (bind _returnvalue (CallFunction|CreateAssetNameString (Variables|ItemDefinition|GetCharacter)))
                  (Rendering|Components|TextRender|SetText (Variables|Default|GetTextRender) _returnvalue)))
              (:then_1
                (if (Variables|ItemDefinition|GetShowWrapID)
                  (bind _returnvalue_2 (CallFunction|CreateAssetNameString (Variables|ItemDefinition|GetWrap)))
                  (Rendering|Components|TextRender|SetText (Variables|Default|GetTextRender) _returnvalue_2)))
              (:then_2
                (if (Variables|ItemDefinition|GetShowPickaxeID)
                  (bind _returnvalue_3 (CallFunction|CreateAssetNameString (Variables|ItemDefinition|GetPickaxe)))
                  (Rendering|Components|TextRender|SetText (Variables|Default|GetTextRender) _returnvalue_3)))
              (:then_3
                (if (Variables|ItemDefinition|GetShowGliderID)
                  (bind _returnvalue_4 (CallFunction|CreateAssetNameString (Variables|ItemDefinition|GetGlider)))
                  (Rendering|Components|TextRender|SetText (Variables|Default|GetTextRender) _returnvalue_4)))
              (:then_4
                (if (Variables|ItemDefinition|GetShowBackpackID)
                  (bind _returnvalue_5 (CallFunction|CreateAssetNameString (Variables|ItemDefinition|GetBackpack)))
                  (Rendering|Components|TextRender|SetText (Variables|Default|GetTextRender) _returnvalue_5)))
              (:then_5
                (if (Variables|ItemDefinition|GetRemoveBackpack)
                  _))
              (:then_6
                (if (not (Variables|Default|GetIsVariant))
                  (Components|DestroyComponent (Variables|Default|GetInteractBox))))
              (:then_7
                (if (not (Variables|ItemDefinition|ForcedVarianceColor|GetApplyForcedVarianceColor))
                  (Components|DestroyComponent (Variables|Default|GetBackgroundWall)))))
            (Rendering|SetVisibility (Variables|Default|GetInteractionButton) true true)
            (Utilities|FlowControl|Sequence
              (:then_0
                (Components|DestroyComponent (Variables|Default|GetCollisionTrigger)))
              (:then_1
                (Components|DestroyComponent (Variables|Default|GetTextRender)))
              (:then_2
                (Components|DestroyComponent (Variables|Default|GetPlaque)))
              (:then_3
                (Rendering|SetVisibility (Variables|Default|GetBillboard) false false))
              (:then_4
                (if (Variables|ItemDefinition|GetRemoveBackpack)
                  (Rendering|Components|TextRender|SetVerticalAlignment (Variables|Default|GetVariantTextRender) "EVRTA_TextCenter")
                  (Rendering|Components|TextRender|SetVertSpacingAdjust (Variables|Default|GetVariantTextRender) -9.0)
                  (Transformation|AddLocalOffset (Variables|Default|GetVariantTextRender) "0, 0,18.000000" false false)
                  (Rendering|Components|TextRender|SetText (Variables|Default|GetVariantTextRender) (Utilities|Text|MakeLiteralText "NSLOCTEXT(\"[190B08FA4B364C2A6544B5BB1A6B5D22]\", \"D10B92F0499296CD3556C985EDF4D220\", \"Remove\\r\\nBackpack\")"))
                  (else
                    (Rendering|Components|TextRender|SetText (Variables|Default|GetVariantTextRender) (Utilities|Text|ToText(String) (Utilities|String|Replace (GameplayTags|GetDebugStringfromGameplayTag (LocalVariables|GetL_ActiveVariantTag)) "Cosmetics.Variant.Property."))))))))))
      (Variables|Default|SetShowCharacterID false)
      (goto _label))
    (:then_1
      (if _applyforcedvariancecolor
        (Rendering|Material|SetColorParameterValueonMaterials (Variables|Default|GetStandingBase) "BaseColor" (select _applyforcedvariancecolor (Math|Conversions|ToLinearColor(Color) (Variables|ItemDefinition|ForcedVarianceColor|GetForcedVarianceColor)) 0))))))
