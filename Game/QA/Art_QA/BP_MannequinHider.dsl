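; Interactable lever switch for Art QA that toggles in-game visibility of every SoldierMannequin_Male actor in the world.
; Each interaction calls Activate Switch, where a FlipFlop sets IsHidden.
; A Sequence first plays the Switch On or Switch Off timeline to rotate the lever.
; It then finds all mannequins by class and applies IsHidden through SetActorHiddenInGame.
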(event Interaction|EventBlueprintOnInteract
  (params (InteractingPawn :type Fort%20Pawn%20Object%20Reference) (InteractionBeingAttempted :type EInteractionBeingAttempted%20Enum :default "FirstInteraction"))
  (CallFunction|ActivateSwitch))

(event Custom|Activate%20Switch
  (bind _isa (Utilities|FlowControl|FlipFlop)
    (:A
      _)
    (:B
      _))
  (bind _output_get (Variables|Default|SetIsHidden _isa))
  (Utilities|FlowControl|Sequence
    (:then_0
      (if _output_get
        (bind (_direction _switch) (Timeline|Switch%20On Play)
          (:Update
            (Transformation|SetRelativeRotation (Variables|Default|GetLever) (Math|Rotator|MakeRotator 0.0 (Math|Float|Lerp -5e+01 50 _switch)) false false)))
        (else
          (bind (_direction_2 _switchoff) (Timeline|Switch%20Off Play)
            (:Update
              (Transformation|SetRelativeRotation (Variables|Default|GetLever) (Math|Rotator|MakeRotator 0.0 (Math|Float|Lerp -5e+01 5e+01 _switchoff)) false false))))))
    (:then_1
      (bind _outactors (Actor|GetAllActorsOfClass "/Game/Characters/Player/Mannequins/SoldierMannequin_Male.SoldierMannequin_Male_C"))
      (for _array_element _outactors
        (Rendering|SetActorHiddenInGame _array_element (Variables|Default|GetIsHidden))))))

(fn BlueprintCanInteract
  (params (InteractingPawn :type Fort%20Pawn%20Object%20Reference) (InteractionBeingAttempted :type EInteractionBeingAttempted%20Enum :default "FirstInteraction") (InteractionType :type TInteractionType%20Enum :default "IT_NoInteraction"))
  (return true))

(fn BlueprintGetInteractionString
  (params (InteractingPawn :type Fort%20Pawn%20Object%20Reference) (InteractionBeingAttempted :type EInteractionBeingAttempted%20Enum :default "FirstInteraction"))
  (return (Utilities|Text|ToText(String) "Hold E to Activate")))

(fn ConstructionScript)
