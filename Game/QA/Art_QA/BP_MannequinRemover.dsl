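; Interactable QA lever switch that removes all male soldier mannequins from the level.
; Interaction calls Activate Switch, whose sequence plays a timeline to rotate the lever, then destroys every SoldierMannequin_Male actor found.
; CanInteract gates the prompt and is cleared inside the destroy loop, so the switch stays usable if no mannequins are found.
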(event Interaction|EventBlueprintOnInteract
  (params (InteractingPawn :type Fort%20Pawn%20Object%20Reference) (InteractionBeingAttempted :type EInteractionBeingAttempted%20Enum :default "FirstInteraction"))
  (CallFunction|ActivateSwitch))

(event Custom|Activate%20Switch
  (Utilities|FlowControl|Sequence
    (:then_0
      (bind (_direction _switch) (Timeline|Switch%20On Play)
        (:Update
          (Transformation|SetRelativeRotation (Variables|Default|GetLever) (Math|Rotator|MakeRotator 0.0 (Math|Float|Lerp -5e+01 50 _switch)) false false))))
    (:then_1
      (bind _outactors (Actor|GetAllActorsOfClass "/Game/Characters/Player/Mannequins/SoldierMannequin_Male.SoldierMannequin_Male_C"))
      (for _array_element _outactors
        (Actor|DestroyActor _array_element)
        (Variables|Default|SetCanInteract false)))))

(fn BlueprintCanInteract
  (params (InteractingPawn :type Fort%20Pawn%20Object%20Reference) (InteractionBeingAttempted :type EInteractionBeingAttempted%20Enum :default "FirstInteraction") (InteractionType :type TInteractionType%20Enum :default "IT_NoInteraction"))
  (return (Variables|Default|GetCanInteract)))

(fn BlueprintGetInteractionString
  (params (InteractingPawn :type Fort%20Pawn%20Object%20Reference) (InteractionBeingAttempted :type EInteractionBeingAttempted%20Enum :default "FirstInteraction"))
  (return (Utilities|Text|ToText(String) "Hold E to Activate")))

(fn ConstructionScript)
