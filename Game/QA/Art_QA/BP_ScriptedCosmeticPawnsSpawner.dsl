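; An interactable QA lever switch that spawns scripted pawns wearing cosmetic loadouts for art review.
; Interaction reads every row of a loadout data table and passes the list to the interacting player's FortCheatManager, which spawns pawns using a location marker, column depth and spacing.
; On remote clients, activation animates the lever with a timeline and flip-flops the On/Off text when the switch has an on state.
; Construction sets the cheat-name label.
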(event Interaction|EventBlueprintOnInteract
  (params (InteractingPawn :type Fort%20Pawn%20Object%20Reference) (InteractionBeingAttempted :type EInteractionBeingAttempted%20Enum :default "FirstInteraction"))
  ;[1 1 1] Manual Interaction
  (CallFunction|ActivateSwitch)
  (CallFunction|SpawnLoadouts InteractingPawn))

(event Custom|RunVisualsOnly
  (bind _text_onoff (Variables|Default|GetText_OnOff))
  (bind _text_onoff_2 (Variables|Default|GetText_OnOff))
  ;[1 1 1] Update the visuals
  (if (Variables|Default|GetbHasActualOnState)
    (Utilities|FlowControl|FlipFlop
      (:A
        (Rendering|Components|TextRender|SetText _text_onoff (Utilities|Text|ToText(String) "On"))
        (Rendering|Components|TextRender|SetTextRenderColor _text_onoff_2 (Utilities|Struct|MakeColor 0 100)))
      (:B
        (Rendering|Components|TextRender|SetText _text_onoff (Utilities|Text|ToText(String) "Off"))
        (Rendering|Components|TextRender|SetTextRenderColor _text_onoff_2 (Utilities|Struct|MakeColor 0 0 100))))
    (else
      (Rendering|SetVisibility (Variables|Default|GetText_OnOff) false false)))
  (bind (_direction _switch) (Timeline|Switch%20On PlayFromStart)
    (:Update
      (Transformation|SetRelativeRotation (Variables|Default|GetLever) (Math|Rotator|MakeRotator 0.0 (Math|Float|Lerp -50 50 _switch)) false false))))

(event Custom|Activate%20Switch
  ;[1 1 1] When activated, do the right things
  (Networking|SwitchHasAuthority
    (:Remote
      (CallFunction|RunVisualsOnly))) ; Spawn AI on server\nAnimate switch on client
)

(event Custom|Switch%20Visually%20Off
  ;[1 1 1] Update the visuals
  (if (Variables|Default|GetbHasActualOnState)
    (Rendering|Components|TextRender|SetText (Variables|Default|GetText_OnOff) (Utilities|Text|ToText(String) "Off"))
    (Rendering|Components|TextRender|SetTextRenderColor (Variables|Default|GetText_OnOff) (Utilities|Struct|MakeColor 0 0 100))
    (else
      (Rendering|SetVisibility (Variables|Default|GetText_OnOff) false false)))
  (Timeline|Switch%20On PlayFromStart))

(event Custom|RunLogicAndVisuals
  ;[1 1 1] When activated, do the right things
  (Networking|SwitchHasAuthority
    (:Remote
      (CallFunction|RunVisualsOnly))) ; Spawn AI on server\nAnimate switch on client
)

(event Custom|SpawnLoadouts
  (params (InteractingPawn :type Fort%20Pawn%20Object%20Reference))
  ;[1 1 1] Spawn Pawns with Loadouts equipped
  (CallFunction|SpawnDataTableLoadouts InteractingPawn (Variables|Loadouts|GetLoadoutsToSpawn)))

(fn BlueprintCanInteract
  (params (InteractingPawn :type Fort%20Pawn%20Object%20Reference) (InteractionBeingAttempted :type EInteractionBeingAttempted%20Enum :default "FirstInteraction") (InteractionType :type TInteractionType%20Enum :default "IT_NoInteraction"))
  (return true))

(fn BlueprintGetInteractionString
  (params (InteractingPawn :type Fort%20Pawn%20Object%20Reference) (InteractionBeingAttempted :type EInteractionBeingAttempted%20Enum :default "FirstInteraction"))
  (return (Utilities|Text|ToText(String) "")))

(fn SpawnDataTableLoadouts
  (params (InteractingPawn :type Fort%20Pawn%20Object%20Reference) (LoadoutsDataTable :type Data%20Table%20Object%20Reference))
  (Utilities|FlowControl|Sequence
    (:then_0
      ;[1 1 1] Gather the Loadouts to spawn from the data table
      (bind _outrownames (DataTable|GetDataTableRowNames LoadoutsDataTable))
      (for _array_element _outrownames
        (bind _returnvalue (Utilities|GetDataTableRow LoadoutsDataTable _array_element))
        (Utilities|Array|Add (LocalVariables|GetLocal_CosmeticLoadoutsToSpawn) (Utilities|Struct|BreakFortCosmeticLoadoutTableRow _returnvalue))))
    (:then_1
      ;[1 1 1] Spawn the list of loadouts
      (Utilities|IsValid InteractingPawn
        (:Is%20Valid
          (bind _asplayer_controller (Utilities|Casting|CastToPlayerController (Pawn|GetController InteractingPawn)))
          (bind _asfort_cheat_manager (Utilities|Casting|CastToFortCheatManager (Variables|/Script/Engine.PlayerController|GetCheatManager _asplayer_controller)))
          (Class|FortCheatManager|SpawnScriptedPawnswithCosmeticLoadouts _asfort_cheat_manager "0" (Variables|FormationData|GetFormationLocationMarker) (LocalVariables|GetLocal_CosmeticLoadoutsToSpawn) (Variables|FormationData|GetColumnDepth) (Variables|FormationData|GetSpacing) 1e+02 1e+02))))))

(fn ConstructionScript
  (if (Utilities|Text|TextIsEmpty (Variables|Default|GetCheat%20Nice%20Name))
    (Rendering|Components|TextRender|SetText (Variables|Default|GetText_ConsoleCommand) (Utilities|Text|ToText(String) ""))
    (else
      (Rendering|Components|TextRender|SetText (Variables|Default|GetText_ConsoleCommand) (Variables|Default|GetCheat%20Nice%20Name)))))
