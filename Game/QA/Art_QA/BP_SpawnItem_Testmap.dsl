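; BP_SpawnItem_Testmap is an item spawner actor, apparently for QA art test maps, that keeps a chosen Fort world item available as a pickup.
; On BeginPlay it spawns the configured item and quantity at its location and binds to the pickup's OnPickup event.
; When collected, it respawns the picked-up item's definition, forming an endless loop.
; Its construction script handles a text label showing the item name, with a flag for destroying that label.
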(event EventBeginPlay
  (CallFunction|SpawnPickup (Variables|Default|GetItem)))

(event Custom|SpawnPickup
  (params (ItemDefinition :type Fort%20World%20Item%20Definition%20Object%20Reference))
  (bind _returnvalue (Inventory|SpawnPickupInWorld ItemDefinition (Variables|Default|GetQty) (Transformation|GetActorLocation) "0.000000,0.000000,1.000000" 0 false false)) ; Pistol C
  (if (Utilities|IsValid _returnvalue :!class /Script/Engine.KismetSystemLibrary)
    (Interaction|BindEventtoOnPickup _returnvalue (delegate Custom|ItemPickedup))))

(event Custom|ItemPickedup
  (params (SelfActor :type Fort%20Pickup%20Object%20Reference) (InteractingPawn :type Fort%20Pawn%20Object%20Reference))
  (bind _asfort_world_item_definition (Utilities|Casting|CastToFortWorldItemDefinition (Pickup|GetPrimaryItemDefinition SelfActor)))
  (CallFunction|SpawnPickup _asfort_world_item_definition))

(fn ConstructionScript
  (bind _item (Variables|Default|GetItem))
  (if (Utilities|IsValid _item :!class /Script/Engine.KismetSystemLibrary)
    (Rendering|Components|TextRender|SetText (Variables|Default|GetTextRender) (ItemDefinition|GetDisplayName _item))
    (if (Variables|Default|GetDisableTextAndPlatform)
      (Components|DestroyComponent (Variables|Default|GetTextRender)))))
