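; BP_TeleportVolume is an art QA Blueprint volume that moves players who enter it.
; When its Collision component begins an overlap, it casts the other actor to FortPlayerPawn.
; It then teleports that pawn to the Teleport Offset variable, passed directly as the destination with a zero rotation.
; The construction script is empty.
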(event OnComponentBeginOverlap(Collision)
  (params (OverlappedComponent :type Primitive%20Component%20Object%20Reference) (OtherActor :type Actor%20Object%20Reference) (OtherComp :type Primitive%20Component%20Object%20Reference) (OtherBodyIndex :type Integer) (bFromSweep :type Boolean) (SweepResult :type Hit%20Result%20Structure%20%28by%20ref%29))
  (bind _asfort_player_pawn (Utilities|Casting|CastToFortPlayerPawn OtherActor))
  (Transformation|Teleport _asfort_player_pawn (Variables|Default|GetTeleport%20Offset) "0, 0, 0"))

(fn ConstructionScript)
