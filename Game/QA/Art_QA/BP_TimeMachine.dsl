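; BP_TimeMachine is a "Time Dilation Pad" in the Art QA folder that changes global time dilation through cylinder overlap.
; Overlap begin shows game seconds on a TextRender, then after a delay applies the TimeDilation variable and records success in DilationSet.
; Gate and Sequence nodes control that timed update path.
; Overlap end restores normal speed after a delay, resets the label and clears the flag.
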(event OnComponentBeginOverlap(Cylinder)
  (params (OverlappedComponent :type Primitive%20Component%20Object%20Reference) (OtherActor :type Actor%20Object%20Reference) (OtherComp :type Primitive%20Component%20Object%20Reference) (OtherBodyIndex :type Integer) (bFromSweep :type Boolean) (SweepResult :type Hit%20Result%20Structure%20%28by%20ref%29))
  (bind _timedilation (Variables|Default|GetTimeDilation))
  (Utilities|FlowControl|Sequence
    (:then_0
      (Rendering|Components|TextRender|SetText (Variables|Default|GetTextRender) (Utilities|Text|ToText(Float) (Utilities|Time|GetTimeSeconds) "HalfToEven" false true 1 324 0 0))
      (Utilities|FlowControl|Delay 1.0)
      (if (Variables|Default|GetDilationSet)
        (Utilities|FlowControl|Gate true
          (:Exit
            _)))
      (Utilities|Time|SetGlobalTimeDilation _timedilation)
      (Variables|Default|SetDilationSet (== _timedilation (Utilities|Time|GetGlobalTimeDilation)))
      (Utilities|FlowControl|Sequence
        (:then_0
          (Utilities|FlowControl|Gate true
            (:Exit
              _)))
        (:then_1
          _)))
    (:then_1
      _)))

(event OnComponentEndOverlap(Cylinder)
  (params (OverlappedComponent :type Primitive%20Component%20Object%20Reference) (OtherActor :type Actor%20Object%20Reference) (OtherComp :type Primitive%20Component%20Object%20Reference) (OtherBodyIndex :type Integer))
  (bind _timedilation (Variables|Default|GetTimeDilation))
  (Utilities|FlowControl|Sequence
    (:then_0
      (Utilities|FlowControl|Gate true
        (:Exit
          _)))
    (:then_1
      _)
    (:then_2
      (Utilities|FlowControl|Delay 1.0)
      (Utilities|Time|SetGlobalTimeDilation 1.0)
      (Rendering|Components|TextRender|SetText (Variables|Default|GetTextRender) (Utilities|Text|MakeLiteralText "NSLOCTEXT(\"[63DE5FAC4CA3A3C787D023B27E4E4D54]\", \"116B403348A237C2CD58C89C0CC7AE79\", \"Time Dilation Pad\")"))
      (Variables|Default|SetDilationSet false)))
  (Utilities|FlowControl|Gate true
    (:Exit
      (Rendering|Components|TextRender|SetText (Variables|Default|GetTextRender) (Utilities|Text|ToText(Float) (Utilities|Time|GetTimeSeconds) "HalfToEven" false true 1 324 0 0))
      (Utilities|FlowControl|Delay 1.0)
      (if (Variables|Default|GetDilationSet)
        _)
      (Utilities|Time|SetGlobalTimeDilation _timedilation)
      (Variables|Default|SetDilationSet (== _timedilation (Utilities|Time|GetGlobalTimeDilation)))
      (Utilities|FlowControl|Sequence
        (:then_0
          _)
        (:then_1
          _)))))

(fn ConstructionScript)
