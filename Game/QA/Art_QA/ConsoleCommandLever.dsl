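; QA interactable lever that runs a configurable console command, shown on its label and interaction prompt.
; Interaction, or match start when configured, fires the command according to authority and server/client flags, animating the lever and optional On/Off text.
; Additional commands then run at once or one by one through a self-rescheduling timer, printing each and blocking further pulls until finished.
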(event Custom|ExecuteAdditionalConsoleCommands
  (bind _additionalconsolecommands (Variables|Default|GetAdditionalConsoleCommands))
  ;[1 1 1] ExecuteAdditionalConsoleCommands
  ; Runs through the array
  (if (< (Variables|Default|GetAdditionalCommandsRunCount) (Utilities|Array|Length (Variables|Default|GetAdditionalConsoleCommands)))
    (Variables|Default|SetbRunningConsoleCommands true) ; delayed additional commands may REQUIRE specific ordering, \nso additional lever pulls should not be allowed until completed
    (if (> (Variables|Default|GetAdditionalCommandsDelay) 0)
      (bind _returnvalue (Utilities|Time|SetTimerbyEvent (EventDispatchers|CreateEvent :!event Custom|ExecuteNextAdditionalConsoleCommandOnDelay) (Variables|Default|GetAdditionalCommandsDelay)))
      (Variables|Default|SetTimerHandle_AdditionalCommandsDelay _returnvalue)
      (else
        ;;[0.15 0.15 0.15] No delay, run through everything now
        (for _array_element _additionalconsolecommands
          (Development|ExecuteConsoleCommand _array_element)
          (Development|PrintString _array_element true true "(R=0.435002,G=0.828418,B=1.000000,A=1.000000)"))
        (Variables|Default|SetAdditionalCommandsRunCount (Utilities|Array|Length _additionalconsolecommands))
        (Variables|Default|SetbRunningConsoleCommands false)))
    (else
      (Utilities|Time|ClearandInvalidateTimerbyHandle (Variables|Default|GetTimerHandle_AdditionalCommandsDelay))
      (Variables|Default|SetbRunningConsoleCommands false))))

(event Interaction|EventBlueprintOnInteract
  (params (InteractingPawn :type Fort%20Pawn%20Object%20Reference) (InteractionBeingAttempted :type EInteractionBeingAttempted%20Enum :default "FirstInteraction"))
  ;[1 1 1] Manual Interaction
  (CallFunction|ActivateSwitch))

(event Custom|ExecuteNextAdditionalConsoleCommandOnDelay
  (bind _additionalcommandsruncount (Variables|Default|GetAdditionalCommandsRunCount))
  (bind _additionalconsolecommands (Variables|Default|GetAdditionalConsoleCommands))
  (bind _output (Utilities|Array|Get(acopy) _additionalconsolecommands _additionalcommandsruncount))
  ;[1 1 1] ExecuteNextAdditionalConsoleCommandOnDelay
  ; If using delays between commands, need to run the execution one at a time
  (Development|ExecuteConsoleCommand _output)
  (Development|PrintString _output true true "(R=0.435002,G=0.828418,B=1.000000,A=1.000000)")
  (Math|Integer|IncrementInt (Variables|Default|GetAdditionalCommandsRunCount)
    (:%20%20
      (CallFunction|ExecuteAdditionalConsoleCommands))))

(event Custom|Activate%20Switch
  (bind _consolecommand (Variables|Default|GetConsoleCommand))
  ;[0 1 0.007983] When activated, do the right things
  (Variables|Default|SetAdditionalCommandsRunCount 0) ; reset run counter
  (Networking|SwitchHasAuthority
    (:Authority
      (if (Variables|Default|GetbRunOnServer)
        _))
    (:Remote
      (if (Variables|Default|GetbRunsOnClient)
        (Utilities|FlowControl|Sequence
          (:then_0
            (Development|ExecuteConsoleCommand _consolecommand)
            (Development|PrintString _consolecommand)
            (CallFunction|ExecuteAdditionalConsoleCommands))
          (:then_1
            _)))
      (CallFunction|RunVisualsOnly))) ; Spawn AI on server\nAnimate switch on client
)

(event Custom|RunLogicAndVisuals
  (bind _consolecommand (Variables|Default|GetConsoleCommand))
  ;[0 1 0.007983] When activated, do the right things
  (Variables|Default|SetAdditionalCommandsRunCount 0) ; reset run counter
  (Networking|SwitchHasAuthority
    (:Authority
      (if (Variables|Default|GetbRunOnServer)
        _))
    (:Remote
      (if (Variables|Default|GetbRunsOnClient)
        (Utilities|FlowControl|Sequence
          (:then_0
            (Development|ExecuteConsoleCommand _consolecommand)
            (Development|PrintString _consolecommand)
            (CallFunction|ExecuteAdditionalConsoleCommands))
          (:then_1
            _)))
      (CallFunction|RunVisualsOnly))) ; Spawn AI on server\nAnimate switch on client
)

(event EventOnMatchStarted
  ;[1 1 1] Handle Startup
  (if (Variables|Default|GetStartsEnabled)
    (Utilities|FlowControl|DoOnce false
      (:Completed
        (if (Variables|Default|GetbRunAtStart)
          (CallFunction|RunLogicandVisuals)
          (else
            (CallFunction|RunVisualsOnly)))))
    (else
      (CallFunction|SwitchVisuallyOff))))

(event Custom|RunVisualsOnly
  (bind _text_onoff (Variables|Default|GetText_OnOff))
  (bind _text_onoff_2 (Variables|Default|GetText_OnOff))
  ;[1 1 1] Update the visuals
  (if (Variables|Default|GetbHasActualOnState)
    (Utilities|FlowControl|FlipFlop
      (:A
        (Rendering|Components|TextRender|SetText _text_onoff (Utilities|Text|ToText(String) "On"))
        (Rendering|Components|TextRender|SetTextRenderColor _text_onoff_2 (Utilities|Struct|MakeColor 0 100)))
      (:B
        (Rendering|Components|TextRender|SetText _text_onoff (Utilities|Text|ToText(String) "Off"))
        (Rendering|Components|TextRender|SetTextRenderColor _text_onoff_2 (Utilities|Struct|MakeColor 0 0 100))))
    (else
      (Rendering|SetVisibility (Variables|Default|GetText_OnOff) false false)))
  (bind (_direction _switch) (Timeline|Switch%20On PlayFromStart)
    (:Update
      (Transformation|SetRelativeRotation (Variables|Default|GetLever) (Math|Rotator|MakeRotator 0.0 (Math|Float|Lerp -50 50 _switch)) false false))))

(event Custom|Switch%20Visually%20Off
  ;[1 1 1] Update the visuals
  (if (Variables|Default|GetbHasActualOnState)
    (Rendering|Components|TextRender|SetText (Variables|Default|GetText_OnOff) (Utilities|Text|ToText(String) "Off"))
    (Rendering|Components|TextRender|SetTextRenderColor (Variables|Default|GetText_OnOff) (Utilities|Struct|MakeColor 0 0 100))
    (else
      (Rendering|SetVisibility (Variables|Default|GetText_OnOff) false false)))
  (Timeline|Switch%20On PlayFromStart))

(fn BlueprintCanInteract
  (params (InteractingPawn :type Fort%20Pawn%20Object%20Reference) (InteractionBeingAttempted :type EInteractionBeingAttempted%20Enum :default "FirstInteraction") (InteractionType :type TInteractionType%20Enum :default "IT_NoInteraction"))
  (return (not (Variables|Default|GetbRunningConsoleCommands))))

(fn BlueprintGetInteractionString
  (params (InteractingPawn :type Fort%20Pawn%20Object%20Reference) (InteractionBeingAttempted :type EInteractionBeingAttempted%20Enum :default "FirstInteraction"))
  (return (Utilities|Text|ToText(String) (Variables|Default|GetConsoleCommand))))

(fn ConstructionScript
  (if (Utilities|Text|TextIsEmpty (Variables|Default|GetCheat%20Nice%20Name))
    (Rendering|Components|TextRender|SetText (Variables|Default|GetText_ConsoleCommand) (Utilities|Text|ToText(String) (Variables|Default|GetConsoleCommand)))
    (else
      (Rendering|Components|TextRender|SetText (Variables|Default|GetText_ConsoleCommand) (Variables|Default|GetCheat%20Nice%20Name)))))
